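# Niagara effect for the Brawny Bass emote, pouring fish flakes and soft dust puffs from a user-supplied skeletal mesh.
# Both sprite emitters inherit sampled mesh velocity and launch along a mesh-rotated cone.
# Flakes taper off over the loop, User.Colors picks their palette band, and curl noise, drag and repulsion from a second sampled point scatter them.
# Dust is time-of-day tinted, flipbook-animated and fading; BounceIntensity and SpawnIntensity are unused.
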NiagaraSystem /BRCosmetics/Animation/Game/MainPlayer/Emotes/BrawnyBass/FX/NS_Emote_BrawnyBass
dependencies {
 usedDataInterfaces [&NiagaraDataInterfaceCurve, &NiagaraDataInterfaceColorCurve, &NiagaraDataInterfaceArrayFloat]
 usedDynamicInputs []
 usedModules [&SystemState]
}
emitters {
 Dust {
  emitterUpdateScript [
   &EmitterState()
   &SpawnRate(SpawnRate=3, Use Spawn Probability=false) {module=SpawnRate001}
  ]
  particleSpawnScript [
   &InitializeParticle(
    Color=&FNTimeOfDayColor(
     Night Color=<rgba 0.2946 0.4421 0.6927 1>
     Morning Color=<rgba 0.4322 0.4943 0.5365 1>
     Day Color=<rgba 1 1 1 1>
     Evening Color=<rgba 0.6094 0.6094 0.6094 1>
    )
    Lifetime=&RandomRangeFloat(Minimum=1, Maximum=2)
    Sprite Rotation Mode=::NewEnumerator2
    Sprite Size Mode=::NewEnumerator3
    Uniform Sprite Size=&RandomRangeFloat(Minimum=80, Maximum=100)
    UsePositionOffset=false
    Write Material Random=true
   )
   &Color(Color=&MultiplyLinearColor(
     Linear Color=$Particles.Color
     Scale Factor=&MakeVector4_FromLinearColor(LinearColor=&UniformRangedLinearColor(Minimum=<rgba 0.2691 0.1749 0.0864 1>, Maximum=<rgba 0.1753 0.1114 0.0795 1>))
    ))
   &SkeletalMeshLocation(
    Bone Sampling Mode=::NewEnumerator4
    MaskSpawnGroup=false
    Position Sampling=::NewEnumerator1
    Sampled Position Offset=<0 0 5>
    Skeletal Mesh=$User.New Skeletal Mesh
   )
   &InheritVelocity(
    Inherited Velocity Amount Scale=<0.7 0.7 0.7>
    Speed Limit=false
    UseSpeedThreshold=false
    Velocity Source=$Particles.SkeletalMeshLocation.SampledVelocity
   )
   &AddVelocity(
    Cone Axis=&RotateVectorByQuaternion(
     Vector To Rotate=&TransformVector(Vector=<0 -1 1>, Source Space=::World, Destination Space=::Local)
     Quaternion=$Particles.SkeletalMeshLocation.SampledRotation
    )
    Module.UseConeFalloff=true
    Module.UseVelDistribution=true
    Override Seed=false
    Speed Falloff From Cone Axis=0.3
    Velocity Mode=::NewEnumerator2
    Velocity Speed=&RandomRangeFloat(Minimum=60, Maximum=170)
   ) {module=AddVelocity001}
  ]
  particleUpdateScript [
   &ParticleState()
   &ScaleSpriteSize(Uniform Curve Sprite Scale=DataInterface() {propertyValues {Curve {keys [{interpMode=RCIM_Cubic, tangentMode=RCTM_User, time=0, value=0.5}, {interpMode=RCIM_Cubic, tangentMode=RCTM_User, time=1, value=1}]}}})
   &GravityForce(Gravity=&RandomRangeVector(Minimum=<0 0 -100>, Maximum=<0 0 -200>))
   /Niagara/Modules/Update/Forces/CurlNoiseForce(
    Noise Frequency=0.03
    Noise Quality / Cost=::NewEnumerator0
    Noise Strength=&FloatFromCurve(
     FloatCurve=DataInterface() {propertyValues {Curve {keys [
         {arriveTangent=-7.6824, interpMode=RCIM_Cubic, leaveTangent=-7.6824, tangentMode=RCTM_User, time=0, value=2}
         <at 0.15 0.35>
         {interpMode=RCIM_Cubic, tangentMode=RCTM_User, time=1, value=0}
        ]}}}
     Scale Curve=2000
    )
    Pan Noise=true
    Pan Noise Field=<2 2 0>
   ) {enabled=false}
   &AerodynamicDrag(
    Aerodynamic Drag=0.5
    Aerodynamic Rotational Drag=0.5
    Drag Shape Curve=DataInterface() {propertyValues {Curve {keys [
        {interpMode=RCIM_Cubic, tangentMode=RCTM_User, time=0, value=0.1}
        {arriveTangent=2.0799, interpMode=RCIM_Cubic, leaveTangent=2.0799, tangentMode=RCTM_User, time=1, value=1}
       ]}}}
    Lift Shape Curve=DataInterface() {propertyValues {Curve {keys [
        {interpMode=RCIM_Cubic, tangentMode=RCTM_User, time=0, value=0}
        {interpMode=RCIM_Cubic, tangentMode=RCTM_User, time=0.5, value=1}
        {tangentMode=RCTM_User, time=1, value=0}
       ]}}}
    Use Linear Drag=true
    Use Rotational Drag=true
   ) {enabled=false}
   &Drag(Use Linear Drag=true, Use Rotational Drag=false)
   &SolveForcesAndVelocity(Clamp Velocity=false, Limit Acceleration=false)
   &AlignSpriteToMeshOrientation(SetAlignment=true, SetFacing=true) {enabled=false}
   /Niagara/Modules/Update/SubUV/SubUVAnimation(Number Of Frames=63) {module=SubUVAnimation001}
   &ScaleColor(
    Scale Alpha=&FloatFromCurve(
     FloatCurve=DataInterface() {propertyValues {Curve {keys [
         {arriveTangent=14.8786, interpMode=RCIM_Cubic, leaveTangent=14.8786, tangentMode=RCTM_User, time=0, value=0}
         <at 0.1 1 RCIM_Cubic>
         <at 0.9979 0>
        ]}}}
     Scale Curve=0.5
    )
    ScaleA=true
    ScaleRGB=true
   ) {module=ScaleColor001}
  ]
  rendererClasses [&NiagaraSpriteRendererProperties]
  renderers [{rendererClass=&NiagaraSpriteRendererProperties}]
 }
 FishFlakes_Pour {
  emitterUpdateScript [
   &EmitterState()
   &SpawnRate(
    SpawnRate=&FloatFromCurve(
     FloatCurve=DataInterface() {propertyValues {Curve {keys [
         {tangentMode=RCTM_User, time=0, value=1}
         {arriveTangent=-0.0405, interpMode=RCIM_Cubic, leaveTangent=-0.0405, tangentMode=RCTM_User, time=0.5, value=1}
         {arriveTangent=-3.6609, interpMode=RCIM_Cubic, leaveTangent=-3.6609, tangentMode=RCTM_User, time=1, value=0}
        ]}}}
     CurveIndex=$Emitter.NormalizedLoopAge
     Scale Curve=100
    )
    Use Spawn Probability=false
   ) {module=SpawnRate001}
  ]
  particleSpawnScript [
   &InitializeParticle(
    Color=&ColorFromCurve(
     ColorCurve=DataInterface() {propertyValues {
       AlphaCurve {keys [<at 0 1 RCIM_Constant>]}
       BlueCurve {keys [<at 0 0.0319 RCIM_Constant>,<at 0.1837 0.0612 RCIM_Constant>,<at 0.3506 0 RCIM_Constant>,<at 0.5722 0 RCIM_Constant>,<at 0.7996 0 RCIM_Constant>]}
       GreenCurve {keys [<at 0 0.063 RCIM_Constant>,<at 0.1837 0.1221 RCIM_Constant>,<at 0.3506 0.3365 RCIM_Constant>,<at 0.5722 0.1186 RCIM_Constant>,<at 0.7996 0.1341 RCIM_Constant>]}
       RedCurve {keys [<at 0 0.0723 RCIM_Constant>,<at 0.1837 0.3467 RCIM_Constant>,<at 0.3506 0.5728 RCIM_Constant>,<at 0.5722 0.5587 RCIM_Constant>,<at 0.7996 0.0414 RCIM_Constant>]}
      }}
     CurveIndex=&RandomRangeFloat(
      Minimum=&SelectFloatFromArray(
       Direct Array Index=&MakeIntFromFloat(Float=&Add_Float(A=$User.Colors, B=0))
       Direct Array Mode=::NewEnumerator1
       Float Selection Array=DataInterface() {propertyValues {FloatData [0,0.33,0.66]}}
      )
      Maximum=&SelectFloatFromArray(
       Direct Array Index=&MakeIntFromFloat(Float=&Add_Float(A=$User.Colors, B=0))
       Direct Array Mode=::NewEnumerator1
       Float Selection Array=DataInterface() {propertyValues {FloatData [0.32,0.65,1]}}
      )
     )
    )
    Lifetime=&RandomRangeFloat(Minimum=1, Maximum=2)
    Sprite Rotation Mode=::NewEnumerator2
    Sprite Size Mode=::NewEnumerator3
    Uniform Sprite Size=&RandomRangeFloat(Minimum=2, Maximum=8)
    UsePositionOffset=false
    Write Material Random=true
   )
   &SkeletalMeshLocation(
    Bone Sampling Mode=::NewEnumerator4
    MaskSpawnGroup=false
    Position Sampling=::NewEnumerator1
    Sampled Position Offset=<0 0 5>
    Skeletal Mesh=$User.New Skeletal Mesh
   )
   &InheritVelocity(Speed Limit=false, UseSpeedThreshold=false, Velocity Source=$Particles.SkeletalMeshLocation.SampledVelocity)
   &SubUVAnimation(SubUV Animation Mode=::NewEnumerator1)
   &AddVelocity(
    Cone Axis=&RotateVectorByQuaternion(
     Vector To Rotate=&TransformVector(Vector=<0 -1 1>, Source Space=::World, Destination Space=::Local)
     Quaternion=$Particles.SkeletalMeshLocation.SampledRotation
    )
    Module.UseConeFalloff=true
    Module.UseVelDistribution=true
    Override Seed=false
    Speed Falloff From Cone Axis=0.3
    Velocity Mode=::NewEnumerator2
    Velocity Speed=&RandomRangeFloat(Minimum=75, Maximum=200)
   ) {module=AddVelocity001}
  ]
  particleUpdateScript [
   &ParticleState()
   &SkeletalMeshLocation(
    Bone / Socket Index=1
    Bone Sampling Mode=::NewEnumerator4
    MaskSpawnGroup=false
    Orientation Sampling=::NewEnumerator1
    Skeletal Mesh=$User.New Skeletal Mesh
    Use Sampled Mesh Attribute as Ribbon ID=false
    Write To Particle Position=false
   ) {module=SkeletalMeshLocation001}
   &GravityForce(Gravity=&RandomRangeVector(Minimum=<0 0 -100>, Maximum=<0 0 -500>))
   &CurlNoiseForce(
    Noise Frequency=0.03
    Noise Quality / Cost=::NewEnumerator0
    Noise Strength=&FloatFromCurve(
     FloatCurve=DataInterface() {propertyValues {Curve {keys [
         {arriveTangent=-7.6824, interpMode=RCIM_Cubic, leaveTangent=-7.6824, tangentMode=RCTM_User, time=0, value=2}
         <at 0.15 0.35>
         {interpMode=RCIM_Cubic, tangentMode=RCTM_User, time=1, value=0}
        ]}}}
     Scale Curve=2000
    )
    Pan Noise=true
    Pan Noise Field=<2 2 0>
   )
   &Drag(Use Linear Drag=true, Use Rotational Drag=false)
   &Collision(
    Age Colliding Particles=false
    Collision Enabled=&SetBoolByFloatComparison(A=$Particles.Age, B=0.075)
    Control Roll On Collision=true
    KillOnCollision=false
    Module.Control Rest State Via Surface Normal=false
    Module.EnableMaxCollisionCount=false
    Module.Randomize Collision Normal Vector=true
    ScaleRateOfRotation=false
   ) {enabled=false}
   &PointAttractionForce(
    Attraction Radius=150
    Attraction Strength=-60
    Attractor Position:NiagaraPosition=&AddVectorToPosition(
     Position=$Particles.SkeletalMeshLocation001.SampledPosition
     Vector=&TransformVector(Vector=<20 10 25>, Source Space=::World, Destination Space=::Local)
    )
    Kill Within Radius=false
    Module.Use Falloff=true
    UseAttractorOffset=false
   )
   &AerodynamicDrag(
    Aerodynamic Drag=0.5
    Aerodynamic Rotational Drag=0.5
    Drag Shape Curve=DataInterface() {propertyValues {Curve {keys [
        {interpMode=RCIM_Cubic, tangentMode=RCTM_User, time=0, value=0.1}
        {arriveTangent=2.0799, interpMode=RCIM_Cubic, leaveTangent=2.0799, tangentMode=RCTM_User, time=1, value=1}
       ]}}}
    Lift Shape Curve=DataInterface() {propertyValues {Curve {keys [
        {interpMode=RCIM_Cubic, tangentMode=RCTM_User, time=0, value=0}
        {interpMode=RCIM_Cubic, tangentMode=RCTM_User, time=0.5, value=1}
        {tangentMode=RCTM_User, time=1, value=0}
       ]}}}
    Use Linear Drag=true
    Use Rotational Drag=true
   )
   &SolveForcesAndVelocity(Clamp Velocity=false, Limit Acceleration=false)
   &ScaleSpriteSize(Uniform Curve Sprite Scale=DataInterface() {propertyValues {Curve {keys [{interpMode=RCIM_Cubic, tangentMode=RCTM_User, time=0.5, value=1}, {interpMode=RCIM_Cubic, tangentMode=RCTM_User, time=1, value=0}]}}})
   &AlignSpriteToMeshOrientation(SetAlignment=true, SetFacing=true)
   &SpriteRotationRate(Rotation Rate=5)
  ]
  rendererClasses [&NiagaraSpriteRendererProperties]
  renderers [{rendererClass=&NiagaraSpriteRendererProperties}]
 }
}
summary {
 systemName=NS_Emote_BrawnyBass
 emitters [
  {emitterName=FishFlakes_Pour, rendererClasses [&NiagaraSpriteRendererProperties]}
  {emitterName=Dust, rendererClasses [&NiagaraSpriteRendererProperties]}
 ]
 userVariables [
  User.BounceIntensity {type=NiagaraFloat, defaultValue=0, description=""}
  User.Colors {type=NiagaraFloat, defaultValue=0, description=""}
  User.New Skeletal Mesh {
   type=NiagaraDataInterfaceSkeletalMesh
   defaultValue {
    struct=NiagaraExt_VariableValue_DataInterface
    value {dataInterface=&self:NiagaraDataInterfaceSkeletalMesh_0}
   }
  }
  User.SpawnIntensity {type=NiagaraFloat, defaultValue=0, description=""}
 ]
}
