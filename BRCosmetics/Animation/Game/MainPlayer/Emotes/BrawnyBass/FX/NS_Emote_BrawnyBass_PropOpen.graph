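# Niagara one-shot effect for the prop-open moment of the BrawnyBass emote, built from three burst sprite emitters.
# AirGlow is a brief soft flash.
# Dust and Fish_Flakes spawn from User.Skeletal Mesh, launch in upward cones, then move under gravity, curl noise and aerodynamic drag.
# User.Colors picks one of three bands in the flakes' stepped color curve, while dust blends brown tints with time-of-day color and fades out.
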NiagaraSystem /BRCosmetics/Animation/Game/MainPlayer/Emotes/BrawnyBass/FX/NS_Emote_BrawnyBass_PropOpen
dependencies {
 usedDataInterfaces [&NiagaraDataInterfaceCurve, &NiagaraDataInterfaceColorCurve, &NiagaraDataInterfaceArrayFloat]
 usedDynamicInputs []
 usedModules [&SystemState, &Timeline]
}
emitters {
 AirGlow {
  emitterUpdateScript [
   &EmitterState()
   &SpawnBurst_Instantaneous(Spawn Count=1, Use Spawn Probability=false)
  ]
  particleSpawnScript [
   &InitializeParticle(
    Color=<rgba 0.4323 0.4323 0.4323 0.2>
    Lifetime=0.25
    Sprite Size Mode=::NewEnumerator3
    Uniform Sprite Size=90
    UsePositionOffset=false
    Write Material Random=true
   ) {module=InitializeParticle001}
   &DynamicMaterialParameters(
    Index 0 Param 1=2
    Module.Param0WriteEnabled=true
    Module.Param1WriteEnabled=true
    Module.Param2WriteEnabled=true
    Module.Param3WriteEnabled=true
   )
   &CameraOffset(Camera Offset Amount=10)
  ]
  particleUpdateScript [
   &ParticleState()
   &ScaleSpriteSize(Scale Factor=&Vector2DFromFloat(Value=&FloatFromCurve(FloatCurve=DataInterface() {propertyValues {Curve {keys [<at 0 1 RCIM_Cubic>]}}})))
   &ScaleColor(
    Scale Alpha=&FloatFromCurve(FloatCurve=DataInterface() {propertyValues {Curve {keys [<at 0 0>,<at 0.25 0.75>,<at 1 0>]}}})
    ScaleA=true
    ScaleRGB=true
   ) {module=ScaleColor001}
  ]
  rendererClasses [&NiagaraSpriteRendererProperties]
  renderers [{rendererClass=&NiagaraSpriteRendererProperties}]
 }
 Dust {
  emitterUpdateScript [
   &EmitterState()
   &SpawnBurst_Instantaneous(Spawn Count=5, Use Spawn Probability=false)
  ]
  particleSpawnScript [
   &InitializeParticle(
    Color=&FNTimeOfDayColor(
     Night Color=<rgba 0.2946 0.4421 0.6927 1>
     Morning Color=<rgba 0.5365 0.5365 0.5365 1>
     Day Color=<rgba 1 1 1 1>
     Evening Color=<rgba 0.6094 0.6094 0.6094 1>
    )
    Lifetime=&UniformRangedFloat(Minimum=1, Maximum=2)
    Position Offset=<0 10 -15>
    Sprite Rotation Mode=::NewEnumerator2
    Sprite Size Mode=::NewEnumerator1
    Uniform Sprite Size Max=100
    Uniform Sprite Size Min=80
    UsePositionOffset=false
    Write Material Random=true
   ) {module=InitializeParticle001}
   &SkeletalMeshLocation(Bone Sampling Mode=::NewEnumerator4, MaskSpawnGroup=false, Skeletal Mesh=$User.Skeletal Mesh)
   &Color(Color=&MultiplyLinearColor(
     Linear Color=$Particles.Color
     Scale Factor=&MakeVector4_FromLinearColor(LinearColor=&UniformRangedLinearColor(Minimum=<rgba 0.2691 0.1749 0.0864 1>, Maximum=<rgba 0.1753 0.1114 0.0795 1>))
    ))
   &AddVelocity(
    Cone Angle=35
    Cone Axis=<0 -0.5 1>
    Inner Cone Angle=20
    Module.UseConeFalloff=true
    Module.UseVelDistribution=true
    Override Seed=false
    Speed Falloff From Cone Axis=0.3
    Velocity Mode=::NewEnumerator2
    Velocity Speed=&RandomRangeFloat(Minimum=150, Maximum=600)
   ) {module=AddVelocity001}
  ]
  particleUpdateScript [
   &ParticleState()
   &ScaleSpriteSize(Uniform Curve Sprite Scale=DataInterface() {propertyValues {Curve {keys [{interpMode=RCIM_Cubic, tangentMode=RCTM_User, time=0, value=0.5}, {interpMode=RCIM_Cubic, tangentMode=RCTM_User, time=1, value=1}]}}})
   &GravityForce(Gravity=&RandomRangeVector(Minimum=<0 0 -100>, Maximum=<0 0 -350>))
   /Niagara/Modules/Update/Forces/CurlNoiseForce(
    Noise Frequency=0.03
    Noise Quality / Cost=::NewEnumerator0
    Noise Strength=&FloatFromCurve(
     FloatCurve=DataInterface() {propertyValues {Curve {keys [
         {arriveTangent=-7.6824, interpMode=RCIM_Cubic, leaveTangent=-7.6824, tangentMode=RCTM_User, time=0, value=2}
         <at 0.15 0.35>
         {interpMode=RCIM_Cubic, tangentMode=RCTM_User, time=1, value=0}
        ]}}}
     Scale Curve=2000
    )
    Pan Noise=true
    Pan Noise Field=<2 2 0>
   )
   &AerodynamicDrag(Aerodynamic Drag=0.5, Aerodynamic Rotational Drag=0.5, Use Linear Drag=true, Use Rotational Drag=true)
   &SolveForcesAndVelocity(Clamp Velocity=false, Limit Acceleration=false)
   &AlignSpriteToMeshOrientation(SetAlignment=true, SetFacing=true)
   /Niagara/Modules/Update/SubUV/SubUVAnimation(Number Of Frames=63)
   &ScaleColor(
    Scale Alpha=&FloatFromCurve(FloatCurve=DataInterface() {propertyValues {Curve {keys [
         {interpMode=RCIM_Cubic, tangentMode=RCTM_User, time=0, value=1}
         {arriveTangent=-1, interpMode=RCIM_Cubic, leaveTangent=-1, tangentMode=RCTM_User, time=0.0954, value=0.4177}
         <at 0.9979 0>
        ]}}})
    ScaleA=true
    ScaleRGB=true
   ) {module=ScaleColor001}
  ]
  rendererClasses [&NiagaraSpriteRendererProperties]
  renderers [{rendererClass=&NiagaraSpriteRendererProperties}]
 }
 Fish_Flakes {
  emitterUpdateScript [
   &EmitterState()
   &SpawnBurst_Instantaneous(Spawn Count=&RandomRangeInt(Minimum=100, Maximum=200), Use Spawn Probability=false)
  ]
  particleSpawnScript [
   &InitializeParticle(
    Color=&ColorFromCurve(
     ColorCurve=DataInterface() {propertyValues {
       AlphaCurve {keys [<at 0 1 RCIM_Constant>]}
       BlueCurve {keys [<at 0 0.0319 RCIM_Constant>,<at 0.1837 0.0612 RCIM_Constant>,<at 0.3506 0 RCIM_Constant>,<at 0.5722 0 RCIM_Constant>,<at 0.7996 0 RCIM_Constant>]}
       GreenCurve {keys [<at 0 0.063 RCIM_Constant>,<at 0.1837 0.1221 RCIM_Constant>,<at 0.3506 0.3365 RCIM_Constant>,<at 0.5722 0.1186 RCIM_Constant>,<at 0.7996 0.1341 RCIM_Constant>]}
       RedCurve {keys [<at 0 0.0723 RCIM_Constant>,<at 0.1837 0.3467 RCIM_Constant>,<at 0.3506 0.5728 RCIM_Constant>,<at 0.5722 0.5587 RCIM_Constant>,<at 0.7996 0.0414 RCIM_Constant>]}
      }}
     CurveIndex=&RandomRangeFloat(
      Minimum=&SelectFloatFromArray(
       Direct Array Index=&MakeIntFromFloat(Float=&Add_Float(A=$User.Colors, B=0))
       Direct Array Mode=::NewEnumerator1
       Float Selection Array=DataInterface() {propertyValues {FloatData [0,0.33,0.66]}}
      )
      Maximum=&SelectFloatFromArray(
       Direct Array Index=&MakeIntFromFloat(Float=&Add_Float(A=$User.Colors, B=0))
       Direct Array Mode=::NewEnumerator1
       Float Selection Array=DataInterface() {propertyValues {FloatData [0.32,0.65,1]}}
      )
     )
    )
    Lifetime=&UniformRangedFloat(Minimum=2.5, Maximum=5)
    Position Offset=<0 10 -15>
    Sprite Rotation Mode=::NewEnumerator2
    Sprite Size Mode=::NewEnumerator3
    Uniform Sprite Size=&FloatFromCurve(
     FloatCurve=DataInterface() {propertyValues {Curve {keys [
         {arriveTangent=-26.9995, interpMode=RCIM_Cubic, leaveTangent=-26.9995, tangentMode=RCTM_User, time=0, value=8}
         {arriveTangent=-6, interpMode=RCIM_Cubic, leaveTangent=-6, time=0.4, value=3.5}
         <at 1 2 RCIM_Cubic>
        ]}}}
     CurveIndex=&UniformRangedFloat(Minimum=0, Maximum=1)
     Scale Curve=1.25
    )
    UsePositionOffset=false
    Write Material Random=true
   ) {module=InitializeParticle001}
   &SkeletalMeshLocation(Bone Sampling Mode=::NewEnumerator4, MaskSpawnGroup=false, Skeletal Mesh=$User.Skeletal Mesh)
   &AddVelocity(
    Cone Axis=<0 -0.5 1>
    Inner Cone Angle=30
    Module.UseConeFalloff=true
    Module.UseVelDistribution=true
    Override Seed=false
    Speed Falloff From Cone Axis=0.3
    Velocity Mode=::NewEnumerator2
    Velocity Speed=&RandomRangeFloat(Minimum=150, Maximum=700)
   ) {module=AddVelocity001}
   &SubUVAnimation(SubUV Animation Mode=::NewEnumerator1)
  ]
  particleUpdateScript [
   &ParticleState()
   &GravityForce(Gravity=&RandomRangeVector(Minimum=<0 0 -100>, Maximum=<0 0 -500>))
   &CurlNoiseForce(
    Noise Frequency=0.03
    Noise Quality / Cost=::NewEnumerator0
    Noise Strength=&FloatFromCurve(
     FloatCurve=DataInterface() {propertyValues {Curve {keys [
         {arriveTangent=-7.6824, interpMode=RCIM_Cubic, leaveTangent=-7.6824, tangentMode=RCTM_User, time=0, value=2}
         <at 0.15 0.35>
         {interpMode=RCIM_Cubic, tangentMode=RCTM_User, time=1, value=0}
        ]}}}
     Scale Curve=2000
    )
    Pan Noise=true
    Pan Noise Field=<2 2 0>
   )
   &Drag(
    Drag=&FloatFromCurve(
     FloatCurve=DataInterface() {propertyValues {Curve {keys [
         {arriveTangent=-0.0126, interpMode=RCIM_Cubic, leaveTangent=-0.0126, tangentMode=RCTM_User, time=0, value=0}
         {arriveTangent=1.1179, interpMode=RCIM_Cubic, leaveTangent=1.1179, tangentMode=RCTM_User, time=0.2042, value=0.6228}
         {arriveTangent=0.0224, interpMode=RCIM_Cubic, leaveTangent=0.0224, tangentMode=RCTM_User, time=1, value=1}
        ]}}}
     Scale Curve=3
    )
    Use Linear Drag=true
    Use Rotational Drag=false
   ) {enabled=false}
   &AerodynamicDrag(Aerodynamic Drag=0.5, Aerodynamic Rotational Drag=0.5, Use Linear Drag=true, Use Rotational Drag=true)
   &SolveForcesAndVelocity(Clamp Velocity=false, Limit Acceleration=false)
   &ScaleColor(
    Scale Alpha=&FloatFromCurve(FloatCurve=DataInterface() {propertyValues {Curve {keys [<at -0.0092 0.0144>,<at 0.0967 0.9767>,<at 1.0011 0.002>]}}})
    ScaleA=true
    ScaleRGB=true
   ) {enabled=false}
   &AlignSpriteToMeshOrientation(SetAlignment=true, SetFacing=true)
   &ScaleSpriteSize(Scale Factor=&Vector2DFromFloat(Value=&FloatFromCurve(FloatCurve=DataInterface() {propertyValues {Curve {keys [{interpMode=RCIM_Cubic, tangentMode=RCTM_User, time=0.7, value=1}, {interpMode=RCIM_Cubic, tangentMode=RCTM_User, time=1, value=0}]}}})))
  ]
  rendererClasses [&NiagaraSpriteRendererProperties]
  renderers [{rendererClass=&NiagaraSpriteRendererProperties}]
 }
}
summary {
 systemName=NS_Emote_BrawnyBass_PropOpen
 emitters [
  {emitterName=AirGlow, rendererClasses [&NiagaraSpriteRendererProperties]}
  {emitterName=Fish_Flakes, rendererClasses [&NiagaraSpriteRendererProperties]}
  {emitterName=Dust, rendererClasses [&NiagaraSpriteRendererProperties]}
 ]
 userVariables [
  User.Colors {type=NiagaraFloat, defaultValue=0, description=""}
  User.NormalizedNotifyProgress {type=NiagaraFloat, defaultValue=0, description=""}
  User.Skeletal Mesh {
   type=NiagaraDataInterfaceSkeletalMesh
   defaultValue {
    struct=NiagaraExt_VariableValue_DataInterface
    value {dataInterface=&self:NiagaraDataInterfaceSkeletalMesh_0}
   }
  }
 ]
}
